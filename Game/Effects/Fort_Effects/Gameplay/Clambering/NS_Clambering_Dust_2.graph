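# Niagara effect for clambering that kicks up dust and debris from a character's skeletal mesh, with speed lines while it moves.
# GPU chunks, soft puffs and vertically stretched dust split bursts between two sampled bones via spawn groups, using gravity and fading or shrinking over life.
# Speed lines spawn by distance travelled around the root bone, tuned by User parameters including local-client energy for alpha.
# A hand-bone debug gnomon emitter is disabled.
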NiagaraSystem /Game/Effects/Fort_Effects/Gameplay/Clambering/NS_Clambering_Dust_2
dependencies {
 usedDataInterfaces [
  &NiagaraDataInterfaceSkeletalMesh
  &NiagaraDataInterfaceCurve
  &NiagaraDataInterfaceVector2DCurve
 ]
 usedDynamicInputs []
 usedModules [
  &SystemState
  &self:SystemSpawnScript.SystemScriptSource.SystemScriptGraph.NiagaraNodeAssignment_0.SetVariables_#1
 ]
}
emitters {
 Chunks001 {
  simTarget=GPUComputeSim
  emitterUpdateScript [
   &EmitterState(Life Cycle Mode=::NewEnumerator1, Loop Behavior=::NewEnumerator1, Loop Delay=0.05, Loop Duration=0.2, UseLoopDelay=false)
   &SpawnBurst_Instantaneous(
    Spawn Count=&Multiply_Int(A=&RandomRangeInt(Minimum=3, Maximum=4), B=1)
    Use Spawn Probability=true
    UseLoopCountLimit=false
   )
  ]
  particleSpawnScript [
   &InitializeParticle(
    Color Maximum=<rgba 0.2083 0.139 0.0696 1>
    Color Minimum=<rgba 0.333 0.2498 0.1665 1>
    Color Mode=::NewEnumerator2
    Lifetime Max=1
    Lifetime Min=0.5
    Lifetime Mode=::NewEnumerator1
    Mass Mode=::NewEnumerator1
    Sprite Rotation Mode=::NewEnumerator2
    Sprite Size Mode=::NewEnumerator1
    Sprite UV Mode=::NewEnumerator3
    Uniform Sprite Size Max=4
    Uniform Sprite Size Min=2
    UsePositionOffset=false
    Write Material Random=true
   )
   &FN_SetSpawnGroup(Number of Spawn Groups=2)
   &SkeletalMeshLocation(Bone / Socket Index=$Particles.SpawnGroup, Bone Sampling Mode=::NewEnumerator4, MaskSpawnGroup=false, Skeletal Mesh=$System.Skeletal Mesh)
   /SetVariables(Particles.Mass:NiagaraFloat=&Lerp_Float(
     A=1
     B=3
     Alpha=&NormalizeFloat(Value=&MakeFloatFromVector2d(Vector 2D=$Particles.SpriteSize), Min=3, Max=9)
    )) {isSetParameters=true}
   &SphereLocation(
    MaskSpawnGroup=false
    Offset=<0 0 -10>
    Override Seed=false
    Sphere Radius=5
    Surface Only=false
    UseNonUniformScale=false
    UseOffset=true
    UseSphereOrientation=false
   )
   &AddVelocity(
    Coordinate Space=::Simulation
    Velocity=&Multiply_Vector(
     A=$User.ForwardVector
     B=&VectorFromFloat(Value=&RandomRangeFloat(Minimum=-75, Maximum=-175))
    )
   ) {enabled=false}
   &AddVelocity(
    Coordinate Space=::World
    Velocity=&MakeVector(
     X=&RandomRangeFloat(Minimum=-50, Maximum=50)
     Y=&RandomRangeFloat(Minimum=-50, Maximum=50)
     Z=&RandomRangeFloat(Minimum=75, Maximum=150)
    )
   ) {module=AddVelocity001}
  ]
  particleUpdateScript [
   &ParticleState()
   &GravityForce(Gravity=<0 0 -600>)
   &SolveForcesAndVelocity(Clamp Velocity=false, Limit Acceleration=false)
   &ScaleSpriteSize(Scale Factor=&Vector2DFromFloat(Value=&FloatFromCurve(FloatCurve=DataInterface() {propertyValues {Curve {keys [<at 0.5 1 RCIM_Cubic>, {arriveTangent=-3.7562, interpMode=RCIM_Cubic, leaveTangent=-3.7562, tangentMode=RCTM_User, time=1, value=0}]}}})))
  ]
  rendererClasses [&NiagaraSpriteRendererProperties]
  renderers [{rendererClass=&NiagaraSpriteRendererProperties}]
 }
 Dust001 {
  emitterUpdateScript [
   &EmitterState(Life Cycle Mode=::NewEnumerator1, Loop Behavior=::NewEnumerator1, Loop Delay=0.05, Loop Duration=0.35, UseLoopDelay=false)
   &SpawnBurst_Instantaneous(Spawn Count=4, Use Spawn Probability=false, UseLoopCountLimit=false)
  ]
  particleSpawnScript [
   &InitializeParticle(
    Color=<rgba 0.7 0.6246 0.567 0.2>
    Lifetime Max=0.7
    Lifetime Min=0.55
    Lifetime Mode=::NewEnumerator1
    Sprite Rotation Mode=::NewEnumerator2
    Sprite Size Mode=::NewEnumerator1
    Sprite UV Mode=::NewEnumerator3
    Uniform Sprite Size Max=35
    Uniform Sprite Size Min=18
    UsePositionOffset=false
    Write Material Random=true
   )
   &FN_SetSpawnGroup(Number of Spawn Groups=2)
   &SkeletalMeshLocation(
    Bone / Socket Index=$Particles.SpawnGroup
    Bone Sampling Mode=::NewEnumerator4
    MaskSpawnGroup=false
    Sampled Position Offset Coordinate Space:ENiagaraSkelMeshTransforms=::NewEnumerator1
    Skeletal Mesh=$System.Skeletal Mesh
   )
   &SphereLocation(
    MaskSpawnGroup=false
    Offset=<0 0 -10>
    Offset Coordinate Space:ENiagaraCoordinateSpace=::World
    Override Seed=false
    Sphere Origin:NiagaraPosition=$Particles.SkeletalMeshLocation.SampledPosition
    Sphere Radius=0
    Surface Only=false
    UseNonUniformScale=false
    UseOffset=true
    UseSphereOrientation=false
   )
   &AddVelocity(
    Rotation Coordinate Space=::World
    Velocity=&RandomRangeVector(Minimum=<30 30 20>, Maximum=<-30 -30 45>)
   )
   /SetVariables(Particles.unlitColor:LinearColor=<rgba 0.15 0.1339 0.1215 0.25>) {isSetParameters=true}
  ]
  particleUpdateScript [
   &ParticleState()
   &Drag(Use Linear Drag=true, Use Rotational Drag=false) {module=Drag001}
   &GravityForce(Gravity=<0 0 -350>)
   &SolveForcesAndVelocity(Clamp Velocity=false, Limit Acceleration=false)
   &ScaleColor(
    Scale Alpha=&FloatFromCurve(
     FloatCurve=DataInterface() {propertyValues {Curve {keys [
         {arriveTangent=-1.7558, interpMode=RCIM_Cubic, time=0, value=1}
         {arriveTangent=-3.0272, interpMode=RCIM_Cubic, leaveTangent=-3.0272, tangentMode=RCTM_User, time=1, value=0}
        ]}}}
     Scale Curve=1.5
    )
    Scale RGB=&VectorFromFloat(Value=&FloatFromCurve(FloatCurve=DataInterface() {propertyValues {Curve {keys [{arriveTangent=1.8145, interpMode=RCIM_Cubic, leaveTangent=4.5795, tangentMode=RCTM_Break, time=0, value=0.4}, <at 0.3387 0.9565>, <at 1 1>]}}}))
    ScaleA=true
    ScaleRGB=false
   ) {enabled=false, module=ScaleColor001}
   &ScaleColor(
    Scale Alpha=&FloatFromCurve(
     FloatCurve=DataInterface() {propertyValues {Curve {keys [
         {interpMode=RCIM_Cubic, leaveTangent=9.8397, tangentMode=RCTM_Break, time=0, value=0}
         {interpMode=RCIM_Cubic, tangentMode=RCTM_Break, time=0.2, value=1}
         {arriveTangent=-1.3793, interpMode=RCIM_Cubic, tangentMode=RCTM_Break, time=1, value=0}
        ]}}}
     Scale Curve=1.5
    )
    ScaleA=true
    ScaleRGB=true
   )
   &ScaleSpriteSize(Uniform Curve Sprite Scale=DataInterface() {propertyValues {Curve {keys [
        {arriveTangent=1.7258, interpMode=RCIM_Cubic, leaveTangent=1.7258, tangentMode=RCTM_User, time=0, value=1}
        {arriveTangent=0.0885, interpMode=RCIM_Cubic, leaveTangent=0.0885, tangentMode=RCTM_User, time=1, value=1.7}
       ]}}}) {module=ScaleSpriteSize001}
   &ScaleSpriteSize(
    Uniform Curve Scale=1.3
    Uniform Curve Sprite Scale=DataInterface() {propertyValues {Curve {keys [
        {arriveTangent=1.0717, interpMode=RCIM_Cubic, leaveTangent=1.4184, tangentMode=RCTM_Break, time=0, value=0.25}
        {interpMode=RCIM_Cubic, leaveTangent=0.0885, tangentMode=RCTM_Break, time=1, value=1}
       ]}}}
   ) {enabled=false}
   &DynamicMaterialParameters(
    Index 0 Param 1=&FloatFromCurve(FloatCurve=DataInterface() {propertyValues {Curve {keys [
         {arriveTangent=0.474, interpMode=RCIM_Cubic, leaveTangent=0.474, tangentMode=RCTM_User, time=0, value=0.5}
         {arriveTangent=0.4938, tangentMode=RCTM_User, time=1, value=1}
        ]}}})
    Index 0 Param 2=1
    Index 0 Param 4=1
    Index 2 Param 1=&FloatFromCurve(FloatCurve=DataInterface() {propertyValues {Curve {keys [
         {arriveTangent=3.7909, interpMode=RCIM_Cubic, leaveTangent=3.7909, tangentMode=RCTM_User, time=0, value=0}
         {arriveTangent=1.8782, interpMode=RCIM_Cubic, leaveTangent=1.8782, tangentMode=RCTM_User, time=0.2, value=0.55}
         {arriveTangent=0.0797, interpMode=RCIM_Cubic, leaveTangent=0.0797, tangentMode=RCTM_User, time=1, value=1}
        ]}}})
    Index 2 Param 2=0.4
    Index 2 Param 3=&FloatFromCurve(FloatCurve=DataInterface() {propertyValues {Curve {keys [
         {arriveTangent=3.7909, interpMode=RCIM_Cubic, leaveTangent=3.7909, tangentMode=RCTM_User, time=0, value=0}
         {arriveTangent=1.8782, interpMode=RCIM_Cubic, leaveTangent=1.8782, tangentMode=RCTM_User, time=0.2, value=0.55}
         {arriveTangent=0.0797, interpMode=RCIM_Cubic, leaveTangent=0.0797, tangentMode=RCTM_User, time=1, value=1}
        ]}}})
    Index 2 Param 4=0.5
    Module.Param0WriteEnabled=true
    Module.Param0WriteEnabled2=true
    Module.Param1WriteEnabled=true
    Module.Param1WriteEnabled2=true
    Module.Param2WriteEnabled=true
    Module.Param2WriteEnabled2=true
    Module.Param3WriteEnabled=true
    Module.Param3WriteEnabled2=true
    Write Parameter Index 2=true
   )
   &CameraOffset(Camera Offset Amount=&RandomRangeFloat(Minimum=15, Maximum=20))
  ]
  rendererClasses [&NiagaraSpriteRendererProperties]
  renderers [
   {rendererClass=&NiagaraSpriteRendererProperties}
   {rendererClass=&NiagaraSpriteRendererProperties, rendererIndex=1}
  ]
 }
 Dust_Long {
  emitterUpdateScript [
   &EmitterState(Life Cycle Mode=::NewEnumerator1, Loop Behavior=::NewEnumerator1, Loop Delay=0.05, Loop Duration=0.2, UseLoopDelay=false)
   &SpawnBurst_Instantaneous(Spawn Count=4, Use Spawn Probability=false, UseLoopCountLimit=false)
   &SpawnRate(SpawnRate=4, Use Spawn Probability=false)
  ]
  particleSpawnScript [
   &InitializeParticle(
    Color=<rgba 0.7 0.6468 0.567 0.2>
    Lifetime Max=1
    Lifetime Min=0.7
    Lifetime Mode=::NewEnumerator1
    Sprite Size Max=<20 120>
    Sprite Size Min=<15 70>
    Sprite Size Mode=::NewEnumerator2
    UsePositionOffset=false
    Write Material Random=true
   )
   &FN_SetSpawnGroup(Number of Spawn Groups=2)
   &SkeletalMeshLocation(
    Bone / Socket Index=$Particles.SpawnGroup
    Bone Sampling Mode=::NewEnumerator4
    MaskSpawnGroup=false
    Orientation Sampling=::NewEnumerator1
    Sampled Position Offset Coordinate Space:ENiagaraSkelMeshTransforms=::NewEnumerator1
    Skeletal Mesh=$System.Skeletal Mesh
   )
   &ShapeLocation(
    MaskSpawnGroup=false
    Offset=<0 0 -10>
    Offset Coordinate Space=::World
    Offset Mode=::NewEnumerator0
    Override Seed=false
    Sphere Radius=4
    UseOwnerScale=true
   ) {module=ShapeLocation001}
   &SpriteFacingAndAlignment(Sprite Alignment=<0 0 1>, WriteAlignment=true, WriteFacing=false)
   /SetVariables(Particles.unlitColor:LinearColor=<rgba 0.15 0.1339 0.1215 0.25>) {isSetParameters=true}
  ]
  particleUpdateScript [
   &ParticleState()
   &DynamicMaterialParameters(
    Index 0 Param 1=&FloatFromCurve(FloatCurve=DataInterface() {propertyValues {Curve {keys [{arriveTangent=0.474, time=0, value=0}, {arriveTangent=0.4938, time=1, value=1}]}}})
    Index 0 Param 2=1
    Index 0 Param 4=1
    Index 2 Param 1=&FloatFromCurve(FloatCurve=DataInterface() {propertyValues {Curve {keys [
         {arriveTangent=3.7909, interpMode=RCIM_Cubic, leaveTangent=3.7909, tangentMode=RCTM_User, time=0, value=0}
         {arriveTangent=1.8782, interpMode=RCIM_Cubic, leaveTangent=1.8782, tangentMode=RCTM_User, time=0.2, value=0.55}
         {arriveTangent=0.0797, interpMode=RCIM_Cubic, leaveTangent=0.0797, tangentMode=RCTM_User, time=1, value=1}
        ]}}})
    Index 2 Param 2=0.4
    Index 2 Param 3=&FloatFromCurve(FloatCurve=DataInterface() {propertyValues {Curve {keys [
         {arriveTangent=3.7909, interpMode=RCIM_Cubic, leaveTangent=3.7909, tangentMode=RCTM_User, time=0, value=0}
         {arriveTangent=1.8782, interpMode=RCIM_Cubic, leaveTangent=1.8782, tangentMode=RCTM_User, time=0.2, value=0.55}
         {arriveTangent=0.0797, interpMode=RCIM_Cubic, leaveTangent=0.0797, tangentMode=RCTM_User, time=1, value=1}
        ]}}})
    Index 2 Param 4=0.5
    Module.Param0WriteEnabled=true
    Module.Param0WriteEnabled2=true
    Module.Param1WriteEnabled=true
    Module.Param1WriteEnabled2=true
    Module.Param2WriteEnabled=true
    Module.Param2WriteEnabled2=true
    Module.Param3WriteEnabled=true
    Module.Param3WriteEnabled2=true
    Write Parameter Index 2=true
   )
   &GravityForce(Gravity=&RandomRangeVector(Minimum=<0 0 -5>, Maximum=<0 0 -20>))
   &SolveForcesAndVelocity(Clamp Velocity=false, Limit Acceleration=false)
   &ScaleColor(
    Scale Alpha=&FloatFromCurve(
     FloatCurve=DataInterface() {propertyValues {Curve {keys [
         {arriveTangent=9.843, interpMode=RCIM_Cubic, leaveTangent=9.843, tangentMode=RCTM_User, time=0, value=0}
         {arriveTangent=0.0001, interpMode=RCIM_Cubic, leaveTangent=0.0001, tangentMode=RCTM_User, time=0.25, value=0.9847}
         {arriveTangent=-0.0405, interpMode=RCIM_Cubic, leaveTangent=-0.0405, tangentMode=RCTM_User, time=1, value=0}
        ]}}}
     Scale Curve=1.5
    )
    ScaleA=true
    ScaleRGB=true
   )
   &ScaleSpriteSize(
    Non-Uniform Curve Sprite Scale=DataInterface() {propertyValues {
      XCurve {keys [{tangentMode=RCTM_User, time=0.5, value=1}, {tangentMode=RCTM_User, time=1, value=1}]}
      YCurve {keys [
        {arriveTangent=2.1757, interpMode=RCIM_Cubic, leaveTangent=2.1757, tangentMode=RCTM_User, time=0, value=0.2}
        {interpMode=RCIM_Cubic, tangentMode=RCTM_User, time=1, value=1}
       ]}
     }}
    Scale Sprite Size Mode=::NewEnumerator3
   )
   &CameraOffset(Camera Offset Amount=&RandomRangeFloat(Minimum=15, Maximum=20))
  ]
  rendererClasses [&NiagaraSpriteRendererProperties]
  renderers [
   {rendererClass=&NiagaraSpriteRendererProperties}
   {rendererClass=&NiagaraSpriteRendererProperties, rendererIndex=1}
  ]
 }
 FN_DebugGnomon {
  bEnabled=false
  emitterUpdateScript [
   &EmitterState(
    Life Cycle Mode=::NewEnumerator1
    Loop Behavior=::NewEnumerator1
    Loop Duration Mode=::NewEnumerator1
    Scalability Mode=::NewEnumerator1
    UseLoopDelay=false
   )
   &SpawnBurst_Instantaneous(Use Spawn Probability=false, UseLoopCountLimit=false)
  ]
  particleSpawnScript [
   &InitializeParticle(
    Color Mode=::NewEnumerator0
    Lifetime=5
    Mesh Scale Mode=::NewEnumerator3
    Mesh Uniform Scale=5
    UsePositionOffset=false
    Write Material Random=true
   )
   &SkeletalMeshLocation(
    Bone Sampling Mode=::NewEnumerator4
    MaskSpawnGroup=false
    Sampled Position Offset Coordinate Space:ENiagaraSkelMeshTransforms=::NewEnumerator1
    Skeletal Mesh=DataInterface() {propertyValues {
      DefaultMesh=null
      ExcludeBoneName=root
      OptionalComponentClass=null
      PreviewMesh=&M_Med_Soldier_04_Preview
      SkinningMode=SkinOnTheFly
      SoftSourceActor=null
      SourceMode=Default
      UvSetIndex=0
      WholeMeshLOD=-1
      bExcludeBone=false
      bReadDeformedGeometry=true
      bRequireCurrentFrameData=false
      ComponentTags []
      FilteredBones [hand_l,hand_r]
      FilteredSockets []
      MeshUserParameter {parameter {name=None, typeDefHandle {registeredTypeIndex=1}, varData []}}
      SamplingRegions []
     }}
   )
  ]
  particleUpdateScript [&ParticleState(Kill Particles When Lifetime Has Elapsed=false)]
  rendererClasses [&NiagaraMeshRendererProperties]
  renderers [{rendererClass=&NiagaraMeshRendererProperties, rendererIndex=0}]
 }
 SpeedLine_Particles {
  emitterUpdateScript [
   &EmitterState(Life Cycle Mode=::NewEnumerator1, Loop Behavior=::NewEnumerator1, Loop Delay=0.3, Loop Duration=0.2, UseLoopDelay=false)
   &SpawnRate(
    SpawnRate=&MakeCustomFloatFromBool(Bool=$User.Jumping, True Float=30, False Float=0)
    Use Spawn Probability=false
   ) {enabled=false}
   &SpawnPerUnit(
    Max Movement Threshold=2000
    Module.Use Max Movement Threshold=true
    Module.Use Movement Tolerance=false
    Spawn Spacing=9
    Use Spawn Probability=false
   )
  ]
  particleSpawnScript [
   &InitializeParticle(
    Color=&RandomRangeLinearColor(Minimum=<rgba 1 1 1 0.02>, Maximum=<rgba 1 1 1 0.03>)
    Lifetime=&Multiply_Float(A=&RandomRangeFloat(Minimum=0.2, Maximum=0.3), B=$User.SpeedLineLifetimeMultiplier)
    Sprite Size Max=<1 65>
    Sprite Size Min=<1 37>
    Sprite Size Mode=::NewEnumerator2
    UsePositionOffset=false
    Write Material Random=true
   )
   &SkeletalMeshLocation(
    MaskSpawnGroup=false
    OverrideRandomSeed=false
    Skeletal Mesh=DataInterface() {propertyValues {
      DefaultMesh=null
      ExcludeBoneName=root
      OptionalComponentClass=null
      PreviewMesh=&M_Med_Soldier_04_Preview
      SkinningMode=SkinOnTheFly
      SoftSourceActor=null
      SourceMode=Default
      UvSetIndex=0
      WholeMeshLOD=-1
      bExcludeBone=false
      bReadDeformedGeometry=true
      bRequireCurrentFrameData=true
      ComponentTags []
      FilteredBones [root]
      FilteredSockets []
      MeshUserParameter {parameter {name=None, typeDefHandle {registeredTypeIndex=1}, varData []}}
      SamplingRegions []
     }}
   )
   &ShapeLocation(
    MaskSpawnGroup=false
    Offset=<0 0 20>
    Offset Mode=::NewEnumerator0
    Override Seed=false
    Shape Origin:Vector3f=$StackContext.SkeletalMeshLocation.SampledPosition
    Sphere Radius=$User.SphereRadius
    UseOwnerScale=true
   )
   &AddVelocity(Velocity=&RandomRangeVector(Minimum=<0 0 -800>, Maximum=<0 0 -1100>)) {enabled=false}
  ]
  particleUpdateScript [
   &InheritVelocity(
    Inherited Velocity Amount Scale=&VectorFromFloat(Value=-0.1)
    Inherited Velocity Speed Limit=50
    Speed Limit=true
    UseSpeedThreshold=false
   )
   &InheritSourceMovement(
    Applied Position Scale=&VectorFromFloat(Value=0.5)
    ApplyForce=false
    ApplyPosition=true
    ApplyPositionFalloff=false
   ) {enabled=false}
   &SolveForcesAndVelocity(Clamp Velocity=false, Limit Acceleration=false)
   &ParticleState()
   &ScaleColor(
    Scale Alpha=&FloatFromCurve(
     FloatCurve=DataInterface() {propertyValues {Curve {keys [
         {interpMode=RCIM_Cubic, leaveTangent=12.945, tangentMode=RCTM_Break, time=0, value=0}
         {interpMode=RCIM_Cubic, tangentMode=RCTM_Break, time=0.2, value=1}
         {arriveTangent=-1.8151, interpMode=RCIM_Cubic, tangentMode=RCTM_Break, time=1, value=0}
        ]}}}
     Scale Curve=&Multiply_Float(
      A=&MakeCustomFloatFromBool(Bool=$User.bIsLocalClient, True Float=$User.energy, False Float=1)
      B=$User.SpeedLineAlphaMultiplier
     )
    )
    Scale RGB=<0.8 0.8 0.8>
    ScaleA=true
    ScaleRGB=true
   )
   &ScaleSpriteSize(
    Non-Uniform Curve Scale=&MakeVector2D(X=$User.SpeedLineThickness, Y=1)
    Non-Uniform Curve Sprite Scale=DataInterface() {propertyValues {
      XCurve {keys [{tangentMode=RCTM_User, time=0.55, value=1}, {tangentMode=RCTM_User, time=1, value=0}]}
      YCurve {keys [
        {interpMode=RCIM_Cubic, leaveTangent=0.9226, tangentMode=RCTM_Break, time=0, value=0.55}
        {arriveTangent=0.0265, interpMode=RCIM_Cubic, tangentMode=RCTM_Break, time=1, value=1}
       ]}
     }}
    Scale Sprite Size Mode=::NewEnumerator3
   )
  ]
  rendererClasses [&NiagaraSpriteRendererProperties]
  renderers [{rendererClass=&NiagaraSpriteRendererProperties}]
 }
}
summary {
 systemName=NS_Clambering_Dust_2
 emitters [
  {emitterName=SpeedLine_Particles, rendererClasses [&NiagaraSpriteRendererProperties]}
  {emitterName=Chunks001, simTarget=GPUComputeSim, rendererClasses [&NiagaraSpriteRendererProperties]}
  {emitterName=Dust001, rendererClasses [&NiagaraSpriteRendererProperties]}
  {emitterName=Dust_Long, rendererClasses [&NiagaraSpriteRendererProperties]}
  {bEnabled=false, emitterName=FN_DebugGnomon, rendererClasses [&NiagaraMeshRendererProperties]}
 ]
 userVariables [
  User.bIsLocalClient {type=NiagaraBool}
  User.energy {type=NiagaraFloat, defaultValue=0, description=""}
  User.ForwardVector {type=Vector3f, description="", defaultValue {struct=Vector3f, value=<0 0 0>}}
  User.Jumping {type=NiagaraBool, defaultValue=true}
  User.SpeedLineAlphaMultiplier {type=NiagaraFloat, defaultValue=0.8, description=""}
  User.SpeedLineLifetimeMultiplier {type=NiagaraFloat, defaultValue=1, description=""}
  User.SpeedLineThickness {type=NiagaraFloat, defaultValue=1, description=""}
  User.SphereRadius {type=NiagaraFloat, defaultValue=17, description=""}
 ]
}
